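# Translucent, unlit, two-sided particle-sprite material in the weapon impact effects folder, built around an electric impact line texture.
# Its only logic multiplies particle color alpha by the green channel of T_Electric_Impact_Line_01.
# The graph shows no connection from that product to a material output such as Emissive or Opacity.
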Material /Figment_S10/Weapons/FORT_Generic_Effects/Impacts/M_MF_Simple_GPU_01
graph {
 properties {
  bUsedWithParticleSprites=true
  blendMode=BLEND_Translucent
  pixelDepthOffsetMode=PDOM_Legacy
  shadingModel=MSM_Unlit
  twoSided=true
  usedShadingModels=Unlit
 }
 nodes [Multiply(A=ParticleColor().A, B=TextureSample() {texture=&/T_Electric_Impact_Line_01}.G)]
}
